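# PBR mesh-particle master material for the Monsoon set in the Dealership shared library.
# Base color, packed specular/roughness/metallic and layered emissive combine panned textures, particle color, Fresnel and a light-direction fake shadow.
# Opacity uses channel-masked erosion, directional dissolve, depth fade, temporal dither and a full-screen transition alpha.
# Two dynamic parameters expose paint, emissive, metallic, dissolve and world position offset controls to particles, with scalar-driven If nodes toggling features.
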Material /DealershipSharedLibrary/Assets/Sets/Monsoon/Material/M_DLR_PBR_MeshParticle_Monsoon
graph {
 expressions {
  MaterialExpressionAdd_0 {
   class=MaterialExpressionAdd
   inputs [
    A {from=MaterialExpressionMultiply_0}
    B {from=MaterialExpressionMultiply_1}
   ]
  }
  MaterialExpressionAdd_1 {
   class=MaterialExpressionAdd
   inputs [
    A {from=MaterialExpressionMultiply_7}
    B {from=MaterialExpressionAdd_12}
   ]
  }
  MaterialExpressionAdd_10 {
   class=MaterialExpressionAdd
   inputs [
    A {from=MaterialExpressionIf_9}
    B {from=MaterialExpressionMultiply_28}
   ]
  }
  MaterialExpressionAdd_11 {
   class=MaterialExpressionAdd
   inputs [
    A {from=MaterialExpressionMultiply_46}
    B {from=MaterialExpressionMultiply_41}
   ]
  }
  MaterialExpressionAdd_12 {
   class=MaterialExpressionAdd
   inputs [
    A {from=MaterialExpressionMultiply_8}
    B {from=MaterialExpressionAppendVector_6}
   ]
  }
  MaterialExpressionAdd_13 {
   class=MaterialExpressionAdd
   inputs [
    A {from=""}
    B {from=""}
   ]
  }
  MaterialExpressionAdd_14 {
   class=MaterialExpressionAdd
   inputs [
    A {from=MaterialExpressionIf_17}
    B {from=MaterialExpressionMultiply_51}
   ]
  }
  MaterialExpressionAdd_15 {
   class=MaterialExpressionAdd
   inputs [
    A {from=MaterialExpressionPower_2}
    B {from=MaterialExpressionScalarParameter_37}
   ]
  }
  MaterialExpressionAdd_16 {
   class=MaterialExpressionAdd
   inputs [
    A {from=MaterialExpressionStaticSwitchParameter_0}
    B {from=MaterialExpressionNamedRerouteDeclaration_35}
   ]
  }
  MaterialExpressionAdd_17 {
   class=MaterialExpressionAdd
   inputs [
    A {from=MaterialExpressionScalarParameter_53}
    B {from=MaterialExpressionScalarParameter_54}
   ]
  }
  MaterialExpressionAdd_18 {
   class=MaterialExpressionAdd
   inputs [
    A {from=MaterialExpressionMultiply_43}
    B {from=MaterialExpressionNamedRerouteUsage_24}
   ]
  }
  MaterialExpressionAdd_19 {
   class=MaterialExpressionAdd
   inputs [
    A {from=MaterialExpressionSubtract_4}
    B {from=MaterialExpressionNamedRerouteUsage_43}
   ]
  }
  MaterialExpressionAdd_2 {
   class=MaterialExpressionAdd
   inputs [
    A {from=MaterialExpressionStaticSwitchParameter_0}
    B {from=MaterialExpressionMultiply_12}
   ]
  }
  MaterialExpressionAdd_20 {
   class=MaterialExpressionAdd
   inputs [
    A {from=MaterialExpressionAdd_25}
    B {from=MaterialExpressionMultiply_47}
   ]
  }
  MaterialExpressionAdd_22 {
   class=MaterialExpressionAdd
   inputs [
    A {from=MaterialExpressionScalarParameter_64}
    B {from=MaterialExpressionScalarParameter_65}
   ]
  }
  MaterialExpressionAdd_23 {
   class=MaterialExpressionAdd
   inputs [
    A {from=MaterialExpressionScalarParameter_66}
    B {from=MaterialExpressionScalarParameter_67}
   ]
  }
  MaterialExpressionAdd_24 {
   class=MaterialExpressionAdd
   inputs [
    A {from=MaterialExpressionScalarParameter_68}
    B {from=MaterialExpressionScalarParameter_69}
   ]
   properties {constB=0.01}
  }
  MaterialExpressionAdd_25 {
   class=MaterialExpressionAdd
   inputs [
    A {from=MaterialExpressionMultiply_44}
    B {from=MaterialExpressionMultiply_40}
   ]
  }
  MaterialExpressionAdd_3 {
   class=MaterialExpressionAdd
   inputs [
    A {from=MaterialExpressionMultiply_10}
    B {from=MaterialExpressionMultiply_11}
   ]
  }
  MaterialExpressionAdd_4 {
   class=MaterialExpressionAdd
   inputs [
    A {from=MaterialExpressionMultiply_5}
    B {from=MaterialExpressionMultiply_24}
   ]
  }
  MaterialExpressionAdd_5 {
   class=MaterialExpressionAdd
   inputs [
    A {from=MaterialExpressionDotProduct_0}
    B {from=MaterialExpressionScalarParameter_3}
   ]
  }
  MaterialExpressionAdd_6 {
   class=MaterialExpressionAdd
   inputs [
    A {from=MaterialExpressionMultiply_4}
    B {from=MaterialExpressionClamp_1}
   ]
  }
  MaterialExpressionAdd_7 {
   class=MaterialExpressionAdd
   inputs [
    A {from=MaterialExpressionScalarParameter_22}
    B {from=MaterialExpressionNamedRerouteUsage_8}
   ]
  }
  MaterialExpressionAdd_8 {
   class=MaterialExpressionAdd
   inputs [
    A {from=MaterialExpressionMultiply_26}
    B {from=""}
   ]
   properties {constB=0.5}
  }
  MaterialExpressionAdd_9 {
   class=MaterialExpressionAdd
   inputs [
    A {from=MaterialExpressionMultiply_29}
    B {from=MaterialExpressionMultiply_30}
   ]
  }
  MaterialExpressionAppendVector_0 {
   class=MaterialExpressionAppendVector
   inputs [
    A {from=MaterialExpressionVectorParameter_2, output=Panner X}
    B {from=MaterialExpressionVectorParameter_2, output=Panner X}
   ]
  }
  MaterialExpressionAppendVector_1 {
   class=MaterialExpressionAppendVector
   inputs [
    A {from=MaterialExpressionVectorParameter_3, output=Panner X}
    B {from=MaterialExpressionVectorParameter_3, output=Panner X}
   ]
  }
  MaterialExpressionAppendVector_2 {
   class=MaterialExpressionAppendVector
   inputs [
    A {from=MaterialExpressionVectorParameter_0, output=Panner X}
    B {from=MaterialExpressionVectorParameter_0, output=Panner X}
   ]
  }
  MaterialExpressionAppendVector_3 {
   class=MaterialExpressionAppendVector
   inputs [
    A {from=MaterialExpressionVectorParameter_12, output=Panner X}
    B {from=MaterialExpressionVectorParameter_12, output=Panner X}
   ]
  }
  MaterialExpressionAppendVector_4 {
   class=MaterialExpressionAppendVector
   inputs [
    A {from=MaterialExpressionVectorParameter_16, output=Texture_1_U}
    B {from=MaterialExpressionVectorParameter_16, output=Texture_1_U}
   ]
  }
  MaterialExpressionAppendVector_5 {
   class=MaterialExpressionAppendVector
   inputs [
    A {from=MaterialExpressionVectorParameter_16, output=Texture_2_U}
    B {from=MaterialExpressionVectorParameter_16, output=Texture_2_U}
   ]
  }
  MaterialExpressionAppendVector_6 {
   class=MaterialExpressionAppendVector
   inputs [
    A {from=MaterialExpressionScalarParameter_47, output=""}
    B {from=MaterialExpressionScalarParameter_48, output=""}
   ]
  }
  MaterialExpressionAppendVector_7 {
   class=MaterialExpressionAppendVector
   inputs [
    A {from=MaterialExpressionScalarParameter_43, output=""}
    B {from=MaterialExpressionScalarParameter_44, output=""}
   ]
  }
  MaterialExpressionAtmosphericLightVector_0 {class=MaterialExpressionAtmosphericLightVector, inputs [], properties {}}
  MaterialExpressionCameraVectorWS_1 {class=MaterialExpressionCameraVectorWS, inputs [], properties {}}
  MaterialExpressionCeil_2 {class=MaterialExpressionCeil, inputs [None {from=MaterialExpressionSubtract_0, output=""}], properties {}}
  MaterialExpressionCeil_3 {class=MaterialExpressionCeil, inputs [None {from=MaterialExpressionSubtract_3, output=""}], properties {}}
  MaterialExpressionCeil_4 {class=MaterialExpressionCeil, inputs [None {from=MaterialExpressionLinearInterpolate_26, output=""}], properties {}}
  MaterialExpressionCeil_7 {class=MaterialExpressionCeil, inputs [None {from=MaterialExpressionLinearInterpolate_27, output=""}], properties {}}
  MaterialExpressionChannelMaskParameter_0 {
   class=MaterialExpressionChannelMaskParameter
   inputs [None {from=MaterialExpressionTextureSampleParameter2D_2, output=RGBA}]
   properties {group=02 SRM, parameterName=Specular - Channel, sortPriority=1}
  }
  MaterialExpressionChannelMaskParameter_1 {
   class=MaterialExpressionChannelMaskParameter
   inputs [None {from=MaterialExpressionTextureSampleParameter2D_2, output=RGBA}]
   properties {group=02 SRM, maskChannel=Green, parameterName=Roughness - Channel, sortPriority=2, defaultValue=<rgba 0 1 0 0>}
  }
  MaterialExpressionChannelMaskParameter_2 {
   class=MaterialExpressionChannelMaskParameter
   inputs [None {from=MaterialExpressionTextureSampleParameter2D_3, output=RGBA}]
   properties {group=03 Emissive, parameterName=Emissive - Channel, sortPriority=3}
  }
  MaterialExpressionChannelMaskParameter_3 {
   class=MaterialExpressionChannelMaskParameter
   inputs [None {from=MaterialExpressionTextureSampleParameter2D_4, output=RGBA}]
   properties {group=04 Opacity, parameterName=Dissolve - Channel, sortPriority=31}
  }
  MaterialExpressionChannelMaskParameter_4 {
   class=MaterialExpressionChannelMaskParameter
   inputs [None {from=MaterialExpressionTextureSampleParameter2D_2, output=RGBA}]
   properties {group=02 SRM, maskChannel=Blue, parameterName=Metallic - Channel, sortPriority=3, defaultValue=<rgba 0 0 1 0>}
  }
  MaterialExpressionChannelMaskParameter_5 {
   class=MaterialExpressionChannelMaskParameter
   inputs [None {from=MaterialExpressionMaterialFunctionCall_0, output=Result}]
   properties {group=06 WPO, parameterName=WPODirection}
  }
  MaterialExpressionChannelMaskParameter_6 {
   class=MaterialExpressionChannelMaskParameter
   inputs [None {from=MaterialExpressionTextureSampleParameter2D_11, output=RGBA}]
   properties {group=03 Emissive, parameterName=Emissive - Channel_02, sortPriority=6}
  }
  MaterialExpressionChannelMaskParameter_7 {
   class=MaterialExpressionChannelMaskParameter
   inputs [None {from=MaterialExpressionMaterialFunctionCall_13, output=RGB}]
   properties {group=09 Directional Dissolve, parameterName=DirectionMaskChannel}
  }
  MaterialExpressionChannelMaskParameter_8 {
   class=MaterialExpressionChannelMaskParameter
   inputs [None {from=MaterialExpressionSwitch_0, output=""}]
   properties {group=09 Directional Dissolve, maskChannel=Green, parameterName=TexCoordDirectionMaskChannel, defaultValue=<rgba 0 1 0 0>}
  }
  MaterialExpressionClamp_1 {
   class=MaterialExpressionClamp
   inputs [
    None {from=MaterialExpressionLinearInterpolate_5, output=""}
    Min {from="", output=""}
    Max {from="", output=""}
   ]
   properties {maxDefault=2}
  }
  MaterialExpressionComponentMask_0 {class=MaterialExpressionComponentMask, inputs [None {from=MaterialExpressionVectorParameter_0}], properties {g=true, r=true}}
  MaterialExpressionComponentMask_1 {class=MaterialExpressionComponentMask, inputs [None {from=MaterialExpressionVectorParameter_2}], properties {g=true, r=true}}
  MaterialExpressionComponentMask_2 {class=MaterialExpressionComponentMask, inputs [None {from=MaterialExpressionVectorParameter_3}], properties {g=true, r=true}}
  MaterialExpressionComponentMask_3 {class=MaterialExpressionComponentMask, inputs [None {from=MaterialExpressionTextureCoordinate_5}], properties {r=true}}
  MaterialExpressionComponentMask_4 {class=MaterialExpressionComponentMask, inputs [None {from=MaterialExpressionVectorParameter_12}], properties {g=true, r=true}}
  MaterialExpressionConstant2Vector_0 {class=MaterialExpressionConstant2Vector, properties {g=0.5, r=0.5}}
  MaterialExpressionConstant3Vector_0 {class=MaterialExpressionConstant3Vector}
  MaterialExpressionConstant3Vector_1 {class=MaterialExpressionConstant3Vector, properties {constant=<rgba 0 0 1 0>}}
  MaterialExpressionConstant_0 {class=MaterialExpressionConstant, properties {r=0.25}}
  MaterialExpressionConstant_1 {class=MaterialExpressionConstant}
  MaterialExpressionConstant_2 {class=MaterialExpressionConstant, properties {r=1}}
  MaterialExpressionConstant_3 {class=MaterialExpressionConstant, properties {r=1}}
  MaterialExpressionConstant_4 {class=MaterialExpressionConstant, properties {r=1}}
  MaterialExpressionConstant_5 {class=MaterialExpressionConstant, properties {r=1}}
  MaterialExpressionConstant_6 {class=MaterialExpressionConstant, properties {r=0.5}}
  MaterialExpressionConstant_7 {class=MaterialExpressionConstant, properties {r=0.5}}
  MaterialExpressionConstant_8 {class=MaterialExpressionConstant}
  MaterialExpressionConvert_0 {
   class=MaterialExpressionConvert
   inputs [
    None {from=MaterialExpressionChannelMaskParameter_0, output=R}
    None {from=MaterialExpressionChannelMaskParameter_1, output=R}
    None {from=MaterialExpressionLinearInterpolate_0}
   ]
   properties {
    nodeName=Make Vector3
    convertInputs [
     {
      type=Scalar
      expressionInput {
       expression=&self:MaterialExpressionChannelMaskParameter_0
       inputName=None
       mask=1
       maskA=0
       maskB=0
       maskG=0
       maskR=1
       outputIndex=0
      }
     }
     {
      type=Scalar
      expressionInput {
       expression=&self:MaterialExpressionChannelMaskParameter_1
       inputName=None
       mask=1
       maskA=0
       maskB=0
       maskG=0
       maskR=1
       outputIndex=0
      }
     }
     {
      type=Scalar
      expressionInput {
       expression=&self:MaterialExpressionLinearInterpolate_0
       inputName=None
       mask=0
       maskA=0
       maskB=0
       maskG=0
       maskR=0
       outputIndex=0
      }
     }
    ]
    convertOutputs [{type=Vector3}]
   }
  }
  MaterialExpressionCrossProduct_0 {
   class=MaterialExpressionCrossProduct
   inputs [
    A {from=MaterialExpressionNormalize_0, output=""}
    B {from=MaterialExpressionTransform_0, output=""}
   ]
   properties {}
  }
  MaterialExpressionDesaturation_0 {
   class=MaterialExpressionDesaturation
   inputs [
    None {from=MaterialExpressionIf_6, output=""}
    Fraction {from=MaterialExpressionConstant_3, output=""}
   ]
   properties {}
  }
  MaterialExpressionDotProduct_0 {
   class=MaterialExpressionDotProduct
   inputs [
    A {from=MaterialExpressionLinearInterpolate_9, output=""}
    B {from=MaterialExpressionVertexNormalWS_1, output=""}
   ]
   properties {}
  }
  MaterialExpressionDotProduct_2 {
   class=MaterialExpressionDotProduct
   inputs [
    A {from=MaterialExpressionCameraVectorWS_1, output=""}
    B {from=MaterialExpressionVertexNormalWS_3, output=""}
   ]
   properties {}
  }
  MaterialExpressionDynamicParameter_0 {
   class=MaterialExpressionDynamicParameter
   inputs []
   properties {defaultValue=<rgba 0 1 0 0>, paramNames [Is Painted, Emissive, Metallic, Dissolve]}
  }
  MaterialExpressionDynamicParameter_1 {
   class=MaterialExpressionDynamicParameter
   inputs []
   properties {parameterIndex=1, defaultValue=<rgba 0 1 1 1>, paramNames [WPO, Base Color Emission Amount, Param3, Param4]}
  }
  MaterialExpressionFresnel_0 {
   class=MaterialExpressionFresnel
   inputs [
    ExponentIn {from=MaterialExpressionScalarParameter_50, output=""}
    BaseReflectFractionIn {from="", output=""}
    Normal {from="", output=""}
   ]
   properties {baseReflectFraction=0}
  }
  MaterialExpressionFresnel_2 {
   class=MaterialExpressionFresnel
   inputs [
    ExponentIn {from="", output=""}
    BaseReflectFractionIn {from="", output=""}
    Normal {from="", output=""}
   ]
   properties {baseReflectFraction=0, exponent=2}
  }
  MaterialExpressionIf_0 {
   class=MaterialExpressionIf
   inputs [
    A {from=MaterialExpressionScalarParameter_16, output=""}
    B {from="", output=""}
    "A > B" {from=MaterialExpressionTextureCoordinate_4, output=""}
    "A == B" {from=MaterialExpressionTextureCoordinate_4, output=""}
    "A < B" {from=MaterialExpressionTextureCoordinate_0, output=""}
   ]
   properties {}
  }
  MaterialExpressionIf_1 {
   class=MaterialExpressionIf
   inputs [
    A {from=MaterialExpressionScalarParameter_23, output=""}
    B {from="", output=""}
    "A > B" {from=MaterialExpressionMaterialFunctionCall_2, output=Result}
    "A == B" {from=MaterialExpressionStaticSwitchParameter_3, output=""}
    "A < B" {from=MaterialExpressionStaticSwitchParameter_3, output=""}
   ]
   properties {}
  }
  MaterialExpressionIf_10 {
   class=MaterialExpressionIf
   inputs [
    A {from=MaterialExpressionScalarParameter_39, output=""}
    B {from="", output=""}
    "A > B" {from=MaterialExpressionNamedRerouteDeclaration_23, output=""}
    "A == B" {from=MaterialExpressionConstant_5, output=""}
    "A < B" {from=MaterialExpressionConstant_5, output=""}
   ]
   properties {}
  }
  MaterialExpressionIf_11 {
   class=MaterialExpressionIf
   inputs [
    A {from=MaterialExpressionScalarParameter_49, output=""}
    B {from="", output=""}
    "A > B" {from=MaterialExpressionLinearInterpolate_25, output=""}
    "A == B" {from=MaterialExpressionLinearInterpolate_13, output=""}
    "A < B" {from=MaterialExpressionLinearInterpolate_13, output=""}
   ]
   properties {}
  }
  MaterialExpressionIf_12 {
   class=MaterialExpressionIf
   inputs [
    A {from=MaterialExpressionScalarParameter_51, output=""}
    B {from="", output=""}
    "A > B" {from=MaterialExpressionMultiply_45, output=""}
    "A == B" {from=MaterialExpressionAdd_1, output=""}
    "A < B" {from=MaterialExpressionAdd_1, output=""}
   ]
   properties {}
  }
  MaterialExpressionIf_13 {
   class=MaterialExpressionIf
   inputs [
    A {from=MaterialExpressionScalarParameter_52, output=""}
    B {from="", output=""}
    "A > B" {from=MaterialExpressionChannelMaskParameter_6, output=R}
    "A == B" {from=MaterialExpressionConstant_8, output=""}
    "A < B" {from=MaterialExpressionConstant_8, output=""}
   ]
   properties {}
  }
  MaterialExpressionIf_14 {
   class=MaterialExpressionIf
   inputs [
    A {from=MaterialExpressionScalarParameter_55, output=""}
    B {from="", output=""}
    "A > B" {from=MaterialExpressionNamedRerouteUsage_31, output=""}
    "A == B" {from=MaterialExpressionNamedRerouteUsage_6, output=""}
    "A < B" {from=MaterialExpressionNamedRerouteUsage_6, output=""}
   ]
   properties {}
  }
  MaterialExpressionIf_15 {
   class=MaterialExpressionIf
   inputs [
    A {from=MaterialExpressionScalarParameter_56, output=""}
    B {from="", output=""}
    "A > B" {from=MaterialExpressionMultiply_39, output=""}
    "A == B" {from=MaterialExpressionNamedRerouteUsage_32, output=""}
    "A < B" {from=MaterialExpressionNamedRerouteUsage_32, output=""}
   ]
   properties {}
  }
  MaterialExpressionIf_16 {
   class=MaterialExpressionIf
   inputs [
    A {from=MaterialExpressionScalarParameter_46, output=""}
    B {from="", output=""}
    "A > B" {from=MaterialExpressionAdd_14, output=""}
    "A == B" {from=MaterialExpressionIf_17, output=""}
    "A < B" {from=MaterialExpressionIf_17, output=""}
   ]
   properties {}
  }
  MaterialExpressionIf_17 {
   class=MaterialExpressionIf
   inputs [
    A {from=MaterialExpressionScalarParameter_57, output=""}
    B {from="", output=""}
    "A > B" {from=MaterialExpressionAdd_20, output=""}
    "A == B" {from=MaterialExpressionAdd_25, output=""}
    "A < B" {from=MaterialExpressionAdd_25, output=""}
   ]
   properties {}
  }
  MaterialExpressionIf_2 {
   class=MaterialExpressionIf
   inputs [
    A {from=MaterialExpressionScalarParameter_24, output=""}
    B {from="", output=""}
    "A > B" {from=MaterialExpressionParticleSubUVProperties_0, output=TextureCoordinate0}
    "A == B" {from=MaterialExpressionNamedRerouteDeclaration_18, output=""}
    "A < B" {from=MaterialExpressionNamedRerouteDeclaration_18, output=""}
   ]
   properties {}
  }
  MaterialExpressionIf_3 {
   class=MaterialExpressionIf
   inputs [
    A {from=MaterialExpressionScalarParameter_26, output=""}
    B {from="", output=""}
    "A > B" {from=MaterialExpressionLinearInterpolate_12, output=""}
    "A == B" {from=MaterialExpressionConstant3Vector_0, output=""}
    "A < B" {from=MaterialExpressionConstant3Vector_0, output=""}
   ]
   properties {}
  }
  MaterialExpressionIf_4 {
   class=MaterialExpressionIf
   inputs [
    A {from=MaterialExpressionScalarParameter_30, output=""}
    B {from="", output=""}
    "A > B" {from=MaterialExpressionVectorParameter_10, output=RGB}
    "A == B" {from=MaterialExpressionConvert_0, output=""}
    "A < B" {from=MaterialExpressionConvert_0, output=""}
   ]
   properties {}
  }
  MaterialExpressionIf_5 {
   class=MaterialExpressionIf
   inputs [
    A {from=MaterialExpressionScalarParameter_40, output=""}
    B {from="", output=""}
    "A > B" {from=MaterialExpressionMultiply_31, output=""}
    "A == B" {from=MaterialExpressionTextureSampleParameter2D_0, output=RGB}
    "A < B" {from=MaterialExpressionTextureSampleParameter2D_0, output=RGB}
   ]
   properties {}
  }
  MaterialExpressionIf_6 {
   class=MaterialExpressionIf
   inputs [
    A {from=MaterialExpressionScalarParameter_41, output=""}
    B {from="", output=""}
    "A > B" {from=MaterialExpressionLinearInterpolate_22, output=""}
    "A == B" {from=MaterialExpressionMultiply_3, output=""}
    "A < B" {from=MaterialExpressionMultiply_3, output=""}
   ]
   properties {}
  }
  MaterialExpressionIf_7 {
   class=MaterialExpressionIf
   inputs [
    A {from=MaterialExpressionScalarParameter_34, output=""}
    B {from="", output=""}
    "A > B" {from=MaterialExpressionReroute_2, output=""}
    "A == B" {from=MaterialExpressionNamedRerouteUsage_27, output=""}
    "A < B" {from=MaterialExpressionNamedRerouteUsage_27, output=""}
   ]
   properties {}
  }
  MaterialExpressionIf_8 {
   class=MaterialExpressionIf
   inputs [
    A {from=MaterialExpressionScalarParameter_35, output=""}
    B {from="", output=""}
    "A > B" {from=MaterialExpressionIf_7, output=""}
    "A == B" {from=MaterialExpressionConstant_4, output=""}
    "A < B" {from=MaterialExpressionConstant_4, output=""}
   ]
   properties {}
  }
  MaterialExpressionIf_9 {
   class=MaterialExpressionIf
   inputs [
    A {from=MaterialExpressionScalarParameter_38, output=""}
    B {from="", output=""}
    "A > B" {from=MaterialExpressionReroute_4, output=""}
    "A == B" {from=MaterialExpressionMultiply_37, output=""}
    "A < B" {from=MaterialExpressionMultiply_37, output=""}
   ]
   properties {}
  }
  MaterialExpressionLinearInterpolate_0 {
   class=MaterialExpressionLinearInterpolate
   inputs [
    A {from=MaterialExpressionChannelMaskParameter_4, output=R}
    B {from=""}
    Alpha {from=MaterialExpressionNamedRerouteUsage_10}
   ]
  }
  MaterialExpressionLinearInterpolate_1 {
   class=MaterialExpressionLinearInterpolate
   inputs [
    A {from=MaterialExpressionConstant_2}
    B {from=MaterialExpressionNamedRerouteUsage_17}
    Alpha {from=MaterialExpressionScalarParameter_11}
   ]
   properties {desc=Add fake shadow to emissive}
  }
  MaterialExpressionLinearInterpolate_10 {
   class=MaterialExpressionLinearInterpolate
   inputs [
    A {from=MaterialExpressionConstant3Vector_1}
    B {from=MaterialExpressionTextureSampleParameter2D_1, output=RGB}
    Alpha {from=MaterialExpressionScalarParameter_15}
   ]
  }
  MaterialExpressionLinearInterpolate_11 {
   class=MaterialExpressionLinearInterpolate
   inputs [
    A {from=MaterialExpressionMultiply_21}
    B {from=MaterialExpressionMultiply_19}
    Alpha {from=MaterialExpressionNamedRerouteUsage_21}
   ]
  }
  MaterialExpressionLinearInterpolate_12 {
   class=MaterialExpressionLinearInterpolate
   inputs [
    A {from=MaterialExpressionMultiply_18}
    B {from=MaterialExpressionMultiply_23}
    Alpha {from=MaterialExpressionScalarParameter_21}
   ]
  }
  MaterialExpressionLinearInterpolate_13 {
   class=MaterialExpressionLinearInterpolate
   inputs [
    A {from=MaterialExpressionScalarParameter_32}
    B {from=MaterialExpressionScalarParameter_33}
    Alpha {from=MaterialExpressionFresnel_0}
   ]
  }
  MaterialExpressionLinearInterpolate_14 {
   class=MaterialExpressionLinearInterpolate
   inputs [
    A {from=MaterialExpressionReroute_7}
    B {from=MaterialExpressionReroute_8}
    Alpha {from=MaterialExpressionRound_2}
   ]
  }
  MaterialExpressionLinearInterpolate_15 {
   class=MaterialExpressionLinearInterpolate
   inputs [
    A {from=MaterialExpressionLinearInterpolate_11}
    B {from=MaterialExpressionMultiply_33}
    Alpha {from=MaterialExpressionScalarParameter_28}
   ]
  }
  MaterialExpressionLinearInterpolate_16 {
   class=MaterialExpressionLinearInterpolate
   inputs [
    A {from=MaterialExpressionLinearInterpolate_3}
    B {from=MaterialExpressionMultiply_32}
    Alpha {from=MaterialExpressionScalarParameter_29}
   ]
  }
  MaterialExpressionLinearInterpolate_2 {
   class=MaterialExpressionLinearInterpolate
   inputs [
    A {from=MaterialExpressionMultiply_5}
    B {from=MaterialExpressionAdd_4}
    Alpha {from=MaterialExpressionRound_0}
   ]
   properties {desc=Add base color to emissive}
  }
  MaterialExpressionLinearInterpolate_20 {
   class=MaterialExpressionLinearInterpolate
   inputs [
    A {from=MaterialExpressionChannelMaskParameter_7, output=R}
    B {from=MaterialExpressionChannelMaskParameter_8, output=R}
    Alpha {from=MaterialExpressionScalarParameter_6}
   ]
  }
  MaterialExpressionLinearInterpolate_22 {
   class=MaterialExpressionLinearInterpolate
   inputs [
    A {from=MaterialExpressionMultiply_38}
    B {from=MaterialExpressionMultiply_3}
    Alpha {from=MaterialExpressionNamedRerouteUsage_35}
   ]
  }
  MaterialExpressionLinearInterpolate_23 {
   class=MaterialExpressionLinearInterpolate
   inputs [
    A {from=MaterialExpressionScalarParameter_53}
    B {from=""}
    Alpha {from=MaterialExpressionNamedRerouteUsage_41}
   ]
   properties {constA=1, constB=0}
  }
  MaterialExpressionLinearInterpolate_24 {
   class=MaterialExpressionLinearInterpolate
   inputs [
    A {from=MaterialExpressionAdd_17}
    B {from=""}
    Alpha {from=MaterialExpressionNamedRerouteUsage_41}
   ]
   properties {constA=1, constB=0}
  }
  MaterialExpressionLinearInterpolate_25 {
   class=MaterialExpressionLinearInterpolate
   inputs [
    A {from=MaterialExpressionFresnel_0}
    B {from=""}
    Alpha {from=MaterialExpressionNamedRerouteUsage_30}
   ]
   properties {constB=0}
  }
  MaterialExpressionLinearInterpolate_26 {
   class=MaterialExpressionLinearInterpolate
   inputs [
    A {from=MaterialExpressionNamedRerouteUsage_39}
    B {from=MaterialExpressionScalarParameter_70}
    Alpha {from=MaterialExpressionSubtract_8}
   ]
   properties {constB=8}
  }
  MaterialExpressionLinearInterpolate_27 {
   class=MaterialExpressionLinearInterpolate
   inputs [
    A {from=MaterialExpressionNamedRerouteUsage_39}
    B {from=MaterialExpressionReroute_5}
    Alpha {from=MaterialExpressionSubtract_9}
   ]
   properties {constB=8}
  }
  MaterialExpressionLinearInterpolate_3 {
   class=MaterialExpressionLinearInterpolate
   inputs [
    A {from=MaterialExpressionIf_6}
    B {from=MaterialExpressionMultiply_2}
    Alpha {from=MaterialExpressionNamedRerouteUsage_7}
   ]
  }
  MaterialExpressionLinearInterpolate_4 {
   class=MaterialExpressionLinearInterpolate
   inputs [
    A {from=MaterialExpressionReroute_3}
    B {from=MaterialExpressionOneMinus_4}
    Alpha {from=MaterialExpressionScalarParameter_5}
   ]
  }
  MaterialExpressionLinearInterpolate_5 {
   class=MaterialExpressionLinearInterpolate
   inputs [
    A {from=MaterialExpressionMultiply_13}
    B {from=""}
    Alpha {from=MaterialExpressionReroute_0}
   ]
   properties {constB=0}
  }
  MaterialExpressionLinearInterpolate_6 {
   class=MaterialExpressionLinearInterpolate
   inputs [
    A {from=MaterialExpressionMultiply_36}
    B {from=MaterialExpressionReroute_6}
    Alpha {from=MaterialExpressionRound_2}
   ]
  }
  MaterialExpressionLinearInterpolate_7 {
   class=MaterialExpressionLinearInterpolate
   inputs [
    A {from=MaterialExpressionVectorParameter_5, output=RGB}
    B {from=MaterialExpressionVectorParameter_6, output=RGB}
    Alpha {from=MaterialExpressionSaturate_0}
   ]
  }
  MaterialExpressionLinearInterpolate_8 {
   class=MaterialExpressionLinearInterpolate
   inputs [
    A {from=MaterialExpressionSaturate_0}
    B {from=MaterialExpressionLinearInterpolate_7}
    Alpha {from=MaterialExpressionScalarParameter_12}
   ]
  }
  MaterialExpressionLinearInterpolate_9 {
   class=MaterialExpressionLinearInterpolate
   inputs [
    A {from=MaterialExpressionAtmosphericLightVector_0}
    B {from=MaterialExpressionVectorParameter_7, output=RGB}
    Alpha {from=MaterialExpressionScalarParameter_14}
   ]
   properties {constAlpha=0}
  }
  MaterialExpressionMaterialFunctionCall_0 {
   class=MaterialExpressionMaterialFunctionCall
   inputs [
    UV Coord {from=MaterialExpressionTextureCoordinate_3}
    LenghtScale {from=""}
    RadialScale {from=""}
   ]
   properties {materialFunction=/Game/Global/MaterialFunctions/TechArt/MF_PolarCoordinates}
  }
  MaterialExpressionMaterialFunctionCall_1 {
   class=MaterialExpressionMaterialFunctionCall
   inputs [
    Erosion {from=MaterialExpressionAdd_7}
    Mask {from=MaterialExpressionMultiply_25}
    Edge Width {from=MaterialExpressionScalarParameter_19}
    Width {from=MaterialExpressionConstant_0}
   ]
   properties {materialFunction=/DealershipSharedLibrary/FX/Materials/MaterialFunctions/MF_DLR_Erode}
  }
  MaterialExpressionMaterialFunctionCall_13 {
   class=MaterialExpressionMaterialFunctionCall
   inputs []
   properties {materialFunction=&BoundingBoxBased_0-1_UVW}
  }
  MaterialExpressionMaterialFunctionCall_2 {
   class=MaterialExpressionMaterialFunctionCall
   inputs [
    Alpha Threshold {from=MaterialExpressionStaticSwitchParameter_3}
    Random {from=""}
   ]
   properties {materialFunction=&DitherTemporalAA}
  }
  MaterialExpressionMaterialFunctionCall_3 {
   class=MaterialExpressionMaterialFunctionCall
   inputs [Emissive {from=MaterialExpressionIf_14}]
   properties {materialFunction=&MF_FX_EmissiveAdjustments}
  }
  MaterialExpressionMaterialFunctionCall_4 {
   class=MaterialExpressionMaterialFunctionCall
   inputs [
    Use MPC {from=""}
    Is Sprite {from=""}
    Use World Position Behind Translucency {from=""}
   ]
   properties {materialFunction=/DealershipSharedLibrary/FX/Materials/MaterialFunctions/MF_DLR_FullScreenTransition_Alpha}
  }
  MaterialExpressionMaterialFunctionCall_5 {
   class=MaterialExpressionMaterialFunctionCall
   inputs [
    Opacity In {from=MaterialExpressionStaticSwitchParameter_3}
    DepthFade Distance {from=MaterialExpressionScalarParameter_18}
   ]
   properties {materialFunction=&MF_VCVFX_DepthFade}
  }
  MaterialExpressionMaterialFunctionCall_6 {
   class=MaterialExpressionMaterialFunctionCall
   inputs [Color {from=MaterialExpressionCrossProduct_0}]
   properties {materialFunction=&SplitComponents}
  }
  MaterialExpressionMaterialFunctionCall_7 {
   class=MaterialExpressionMaterialFunctionCall
   inputs [
    R {from=MaterialExpressionMaterialFunctionCall_6, output=G}
    G {from=MaterialExpressionMaterialFunctionCall_6, output=G}
    B {from=""}
    A {from=""}
   ]
   properties {materialFunction=&AppendMany}
  }
  MaterialExpressionMaterialFunctionCall_8 {
   class=MaterialExpressionMaterialFunctionCall
   inputs [Color {from=MaterialExpressionIf_4}]
   properties {materialFunction=&SplitComponents}
  }
  MaterialExpressionMultiply_0 {
   class=MaterialExpressionMultiply
   inputs [
    A {from=MaterialExpressionIf_0}
    B {from=MaterialExpressionComponentMask_0}
   ]
  }
  MaterialExpressionMultiply_1 {
   class=MaterialExpressionMultiply
   inputs [
    A {from=MaterialExpressionAppendVector_2}
    B {from=MaterialExpressionTime_1}
   ]
  }
  MaterialExpressionMultiply_10 {
   class=MaterialExpressionMultiply
   inputs [
    A {from=MaterialExpressionTextureCoordinate_2}
    B {from=MaterialExpressionComponentMask_2}
   ]
  }
  MaterialExpressionMultiply_11 {
   class=MaterialExpressionMultiply
   inputs [
    A {from=MaterialExpressionAppendVector_1}
    B {from=MaterialExpressionTime_2}
   ]
  }
  MaterialExpressionMultiply_12 {
   class=MaterialExpressionMultiply
   inputs [
    A {from=MaterialExpressionNamedRerouteUsage_14}
    B {from=MaterialExpressionScalarParameter_8}
   ]
  }
  MaterialExpressionMultiply_13 {
   class=MaterialExpressionMultiply
   inputs [
    A {from=MaterialExpressionNamedRerouteUsage_14}
    B {from=MaterialExpressionVectorParameter_4, output=RGB}
   ]
  }
  MaterialExpressionMultiply_14 {
   class=MaterialExpressionMultiply
   inputs [
    A {from=MaterialExpressionPower_0}
    B {from=MaterialExpressionScalarParameter_2}
   ]
  }
  MaterialExpressionMultiply_15 {
   class=MaterialExpressionMultiply
   inputs [
    A {from=MaterialExpressionVertexNormalWS_0}
    B {from=MaterialExpressionNamedRerouteUsage_15}
   ]
  }
  MaterialExpressionMultiply_16 {
   class=MaterialExpressionMultiply
   inputs [
    A {from=MaterialExpressionLinearInterpolate_1}
    B {from=MaterialExpressionLinearInterpolate_2}
   ]
  }
  MaterialExpressionMultiply_17 {
   class=MaterialExpressionMultiply
   inputs [
    A {from=MaterialExpressionSmoothStep_0}
    B {from=MaterialExpressionScalarParameter_13}
   ]
  }
  MaterialExpressionMultiply_18 {
   class=MaterialExpressionMultiply
   inputs [
    A {from=MaterialExpressionMultiply_14}
    B {from=MaterialExpressionMultiply_15}
   ]
  }
  MaterialExpressionMultiply_19 {
   class=MaterialExpressionMultiply
   inputs [
    A {from=MaterialExpressionMultiply_16}
    B {from=MaterialExpressionVectorParameter_8, output=RGB}
   ]
  }
  MaterialExpressionMultiply_2 {
   class=MaterialExpressionMultiply
   inputs [
    A {from=MaterialExpressionDesaturation_0}
    B {from=MaterialExpressionParticleColor_0, output=RGB}
   ]
  }
  MaterialExpressionMultiply_20 {
   class=MaterialExpressionMultiply
   inputs [
    A {from=MaterialExpressionMaterialFunctionCall_4, output=Curve - Alpha}
    B {from=MaterialExpressionStaticSwitchParameter_1}
   ]
  }
  MaterialExpressionMultiply_21 {
   class=MaterialExpressionMultiply
   inputs [
    A {from=MaterialExpressionVectorParameter_9, output=RGB}
    B {from=MaterialExpressionScalarParameter_17}
   ]
   properties {constB=4096}
  }
  MaterialExpressionMultiply_22 {
   class=MaterialExpressionMultiply
   inputs [
    A {from=MaterialExpressionStaticSwitchParameter_2}
    B {from=MaterialExpressionParticleColor_2, output=A}
   ]
  }
  MaterialExpressionMultiply_23 {
   class=MaterialExpressionMultiply
   inputs [
    A {from=MaterialExpressionMultiply_15}
    B {from=MaterialExpressionScalarParameter_20}
   ]
  }
  MaterialExpressionMultiply_24 {
   class=MaterialExpressionMultiply
   inputs [
    A {from=MaterialExpressionMultiply_9}
    B {from=MaterialExpressionNamedRerouteUsage_22}
   ]
  }
  MaterialExpressionMultiply_25 {
   class=MaterialExpressionMultiply
   inputs [
    A {from=MaterialExpressionChannelMaskParameter_3, output=R}
    B {from=MaterialExpressionAdd_15}
   ]
  }
  MaterialExpressionMultiply_26 {
   class=MaterialExpressionMultiply
   inputs [
    A {from=MaterialExpressionMaterialFunctionCall_7, output=RG}
    B {from=MaterialExpressionConstant2Vector_0}
   ]
  }
  MaterialExpressionMultiply_27 {
   class=MaterialExpressionMultiply
   inputs [
    A {from=MaterialExpressionNamedRerouteDeclaration_38}
    B {from=MaterialExpressionScalarParameter_31}
   ]
  }
  MaterialExpressionMultiply_28 {
   class=MaterialExpressionMultiply
   inputs [
    A {from=MaterialExpressionTextureSampleParameter2D_6, output=RGB}
    B {from=MaterialExpressionVectorParameter_11, output=RGB}
   ]
  }
  MaterialExpressionMultiply_29 {
   class=MaterialExpressionMultiply
   inputs [
    A {from=MaterialExpressionTextureCoordinate_6}
    B {from=MaterialExpressionComponentMask_4}
   ]
  }
  MaterialExpressionMultiply_3 {
   class=MaterialExpressionMultiply
   inputs [
    A {from=MaterialExpressionPower_5}
    B {from=MaterialExpressionVectorParameter_1, output=RGB}
   ]
  }
  MaterialExpressionMultiply_30 {
   class=MaterialExpressionMultiply
   inputs [
    A {from=MaterialExpressionAppendVector_3}
    B {from=MaterialExpressionTime_3}
   ]
  }
  MaterialExpressionMultiply_31 {
   class=MaterialExpressionMultiply
   inputs [
    A {from=MaterialExpressionTextureSampleParameter2D_0, output=RGB}
    B {from=MaterialExpressionTextureSampleParameter2D_10, output=RGB}
   ]
  }
  MaterialExpressionMultiply_32 {
   class=MaterialExpressionMultiply
   inputs [
    A {from=MaterialExpressionLinearInterpolate_3}
    B {from=MaterialExpressionNamedRerouteUsage_26}
   ]
  }
  MaterialExpressionMultiply_33 {
   class=MaterialExpressionMultiply
   inputs [
    A {from=MaterialExpressionLinearInterpolate_11}
    B {from=MaterialExpressionNamedRerouteUsage_25}
   ]
  }
  MaterialExpressionMultiply_34 {
   class=MaterialExpressionMultiply
   inputs [
    A {from=MaterialExpressionNamedRerouteUsage_37}
    B {from=MaterialExpressionAdd_16}
   ]
  }
  MaterialExpressionMultiply_35 {
   class=MaterialExpressionMultiply
   inputs [
    A {from=MaterialExpressionIf_8}
    B {from=MaterialExpressionLinearInterpolate_16}
   ]
  }
  MaterialExpressionMultiply_36 {
   class=MaterialExpressionMultiply
   inputs [
    A {from=MaterialExpressionParticleColor_5, output=RGB}
    B {from=MaterialExpressionConstant_7}
   ]
  }
  MaterialExpressionMultiply_37 {
   class=MaterialExpressionMultiply
   inputs [
    A {from=MaterialExpressionIf_10}
    B {from=MaterialExpressionLinearInterpolate_15}
   ]
  }
  MaterialExpressionMultiply_38 {
   class=MaterialExpressionMultiply
   inputs [
    A {from=MaterialExpressionNamedRerouteUsage_34}
    B {from=MaterialExpressionConstant_6}
   ]
  }
  MaterialExpressionMultiply_39 {
   class=MaterialExpressionMultiply
   inputs [
    A {from=MaterialExpressionNamedRerouteUsage_36}
    B {from=MaterialExpressionNamedRerouteUsage_33}
   ]
  }
  MaterialExpressionMultiply_4 {
   class=MaterialExpressionMultiply
   inputs [
    A {from=MaterialExpressionAdd_2}
    B {from=MaterialExpressionParticleColor_1, output=RGB}
   ]
  }
  MaterialExpressionMultiply_40 {
   class=MaterialExpressionMultiply
   inputs [
    A {from=MaterialExpressionNamedRerouteUsage_40}
    B {from=MaterialExpressionScalarParameter_71}
   ]
  }
  MaterialExpressionMultiply_41 {
   class=MaterialExpressionMultiply
   inputs [
    A {from=MaterialExpressionAppendVector_7}
    B {from=MaterialExpressionTime_0}
   ]
  }
  MaterialExpressionMultiply_42 {
   class=MaterialExpressionMultiply
   inputs [
    A {from=MaterialExpressionVectorParameter_17, output=RGB}
    B {from=MaterialExpressionAdd_19}
   ]
  }
  MaterialExpressionMultiply_43 {
   class=MaterialExpressionMultiply
   inputs [
    A {from=MaterialExpressionMultiply_42}
    B {from=MaterialExpressionScalarParameter_45}
   ]
  }
  MaterialExpressionMultiply_44 {
   class=MaterialExpressionMultiply
   inputs [
    A {from=MaterialExpressionAdd_18}
    B {from=MaterialExpressionNamedRerouteUsage_42}
   ]
  }
  MaterialExpressionMultiply_45 {
   class=MaterialExpressionMultiply
   inputs [
    A {from=MaterialExpressionTextureCoordinate_9}
    B {from=MaterialExpressionAppendVector_4}
   ]
  }
  MaterialExpressionMultiply_46 {
   class=MaterialExpressionMultiply
   inputs [
    A {from=MaterialExpressionTextureCoordinate_9}
    B {from=MaterialExpressionAppendVector_5}
   ]
  }
  MaterialExpressionMultiply_47 {
   class=MaterialExpressionMultiply
   inputs [
    A {from=MaterialExpressionNamedRerouteUsage_38}
    B {from=MaterialExpressionVectorParameter_8, output=RGB}
   ]
  }
  MaterialExpressionMultiply_5 {
   class=MaterialExpressionMultiply
   inputs [
    A {from=MaterialExpressionAdd_6}
    B {from=MaterialExpressionNamedRerouteUsage_9}
   ]
  }
  MaterialExpressionMultiply_51 {
   class=MaterialExpressionMultiply
   inputs [
    A {from=MaterialExpressionNamedRerouteUsage_28}
    B {from=MaterialExpressionFresnel_2}
   ]
  }
  MaterialExpressionMultiply_6 {
   class=MaterialExpressionMultiply
   inputs [
    A {from=MaterialExpressionNamedRerouteUsage_11}
    B {from=MaterialExpressionNamedRerouteUsage_5}
   ]
  }
  MaterialExpressionMultiply_7 {
   class=MaterialExpressionMultiply
   inputs [
    A {from=MaterialExpressionTextureCoordinate_1}
    B {from=MaterialExpressionComponentMask_1}
   ]
  }
  MaterialExpressionMultiply_8 {
   class=MaterialExpressionMultiply
   inputs [
    A {from=MaterialExpressionAppendVector_0}
    B {from=MaterialExpressionTime_0}
   ]
  }
  MaterialExpressionMultiply_9 {
   class=MaterialExpressionMultiply
   inputs [
    A {from=MaterialExpressionNamedRerouteUsage_13}
    B {from=MaterialExpressionScalarParameter_4}
   ]
  }
  MaterialExpressionNamedRerouteDeclaration_0 {
   class=MaterialExpressionNamedRerouteDeclaration
   inputs [None {from=MaterialExpressionMultiply_35}]
   properties {name=Base Color, nodeColor=<rgba 1 0.2001 0 1>}
  }
  MaterialExpressionNamedRerouteDeclaration_1 {
   class=MaterialExpressionNamedRerouteDeclaration
   inputs [None {from=MaterialExpressionLinearInterpolate_10}]
   properties {name=Normal, nodeColor=<rgba 0.5 0.5 1 1>}
  }
  MaterialExpressionNamedRerouteDeclaration_10 {
   class=MaterialExpressionNamedRerouteDeclaration
   inputs [None {from=MaterialExpressionDynamicParameter_0, output=Is Painted}]
   properties {name=DP Is Painted, nodeColor=<rgba 0.7344 0 1 1>}
  }
  MaterialExpressionNamedRerouteDeclaration_11 {
   class=MaterialExpressionNamedRerouteDeclaration
   inputs [None {from=MaterialExpressionTextureSampleParameter2D_0, output=A}]
   properties {name=Alpha, nodeColor=<rgba 0.4844 0.4844 0.4844 1>}
  }
  MaterialExpressionNamedRerouteDeclaration_12 {
   class=MaterialExpressionNamedRerouteDeclaration
   inputs [None {from=MaterialExpressionMaterialFunctionCall_5, output=Result}]
   properties {name=Opacity, nodeColor=<rgba 1 1 1 1>}
  }
  MaterialExpressionNamedRerouteDeclaration_13 {
   class=MaterialExpressionNamedRerouteDeclaration
   inputs [None {from=MaterialExpressionIf_11}]
   properties {name=Fresnel, nodeColor=<rgba 0.5122 0.2671 0.5521 20>}
  }
  MaterialExpressionNamedRerouteDeclaration_14 {
   class=MaterialExpressionNamedRerouteDeclaration
   inputs [None {from=MaterialExpressionDynamicParameter_1, output=WPO}]
   properties {name=DP WPO, nodeColor=<rgba 0.7344 0 1 1>}
  }
  MaterialExpressionNamedRerouteDeclaration_15 {
   class=MaterialExpressionNamedRerouteDeclaration
   inputs [None {from=MaterialExpressionIf_3}]
   properties {name=WPO, nodeColor=<rgba 0.5938 0 1 1>}
  }
  MaterialExpressionNamedRerouteDeclaration_16 {
   class=MaterialExpressionNamedRerouteDeclaration
   inputs [None {from=MaterialExpressionLinearInterpolate_8}]
   properties {name=FakeShadow, nodeColor=<rgba 0 0.5078 1 1>}
  }
  MaterialExpressionNamedRerouteDeclaration_17 {
   class=MaterialExpressionNamedRerouteDeclaration
   inputs [None {from=MaterialExpressionIf_1}]
   properties {name=Masked Opacity, nodeColor=<rgba 0.4348 0.4348 0.4348 1>}
  }
  MaterialExpressionNamedRerouteDeclaration_18 {
   class=MaterialExpressionNamedRerouteDeclaration
   inputs [None {from=MaterialExpressionAdd_3}]
   properties {name=BC UVs, nodeColor=<rgba 0.7344 0 1 1>}
  }
  MaterialExpressionNamedRerouteDeclaration_19 {
   class=MaterialExpressionNamedRerouteDeclaration
   inputs [None {from=MaterialExpressionMaterialFunctionCall_1, output=Edge}]
   properties {name=Dissolve Edge, nodeColor=<rgba 0.7813 0 1 1>}
  }
  MaterialExpressionNamedRerouteDeclaration_2 {
   class=MaterialExpressionNamedRerouteDeclaration
   inputs [None {from=MaterialExpressionMaterialFunctionCall_8, output=R}]
   properties {name=Specular, nodeColor=<rgba 0.0122 0.0881 1 1>}
  }
  MaterialExpressionNamedRerouteDeclaration_20 {
   class=MaterialExpressionNamedRerouteDeclaration
   inputs [None {from=MaterialExpressionDynamicParameter_1, output=Base Color Emission Amount}]
   properties {name=DP Base Color Emission Amount, nodeColor=<rgba 0.681 0 1 1>}
  }
  MaterialExpressionNamedRerouteDeclaration_21 {
   class=MaterialExpressionNamedRerouteDeclaration
   inputs [None {from=MaterialExpressionPower_3}]
   properties {name=Matcap, nodeColor=<rgba 0.6031 0.1295 0.8646 0.5>}
  }
  MaterialExpressionNamedRerouteDeclaration_22 {
   class=MaterialExpressionNamedRerouteDeclaration
   inputs [None {from=MaterialExpressionLinearInterpolate_20}]
   properties {name=DirectionMask, nodeColor=<rgba 1 0.4922 0 1>}
  }
  MaterialExpressionNamedRerouteDeclaration_23 {
   class=MaterialExpressionNamedRerouteDeclaration
   inputs [None {from=MaterialExpressionLinearInterpolate_14}]
   properties {name=HighlightColor, nodeColor=<rgba 0.1953 1 0 1>}
  }
  MaterialExpressionNamedRerouteDeclaration_24 {
   class=MaterialExpressionNamedRerouteDeclaration
   inputs [None {from=MaterialExpressionSaturate_1}]
   properties {name=CellNoise, nodeColor=<rgba 1 0 0.3984 1>}
  }
  MaterialExpressionNamedRerouteDeclaration_25 {
   class=MaterialExpressionNamedRerouteDeclaration
   inputs [None {from=MaterialExpressionMultiply_28}]
   properties {name=Special Texture, nodeColor=<rgba 1 0.4339 0.1797 1>}
  }
  MaterialExpressionNamedRerouteDeclaration_26 {
   class=MaterialExpressionNamedRerouteDeclaration
   inputs [None {from=MaterialExpressionIf_16}]
   properties {name=Emissive02, nodeColor=<rgba 1 0.2344 0 1>}
  }
  MaterialExpressionNamedRerouteDeclaration_27 {
   class=MaterialExpressionNamedRerouteDeclaration
   inputs [None {from=MaterialExpressionMultiply_13}]
   properties {name=Fresnel_02, nodeColor=<rgba 0 0.8828 1 1>}
  }
  MaterialExpressionNamedRerouteDeclaration_28 {
   class=MaterialExpressionNamedRerouteDeclaration
   inputs [None {from=MaterialExpressionSaturate_12}]
   properties {name=FresnelMask, nodeColor=<rgba 0 0.6484 1 1>}
  }
  MaterialExpressionNamedRerouteDeclaration_29 {
   class=MaterialExpressionNamedRerouteDeclaration
   inputs [None {from=MaterialExpressionSaturate_13}]
   properties {name=FireDissovleMask, nodeColor=<rgba 0 0.6484 1 1>}
  }
  MaterialExpressionNamedRerouteDeclaration_3 {
   class=MaterialExpressionNamedRerouteDeclaration
   inputs [None {from=MaterialExpressionMaterialFunctionCall_8, output=G}]
   properties {name=Roughness, nodeColor=<rgba 0.0122 0.0881 1 1>}
  }
  MaterialExpressionNamedRerouteDeclaration_31 {
   class=MaterialExpressionNamedRerouteDeclaration
   inputs [None {from=MaterialExpressionSaturate_16}]
   properties {name=DirectionDissolve, nodeColor=<rgba 0 1 0.2266 1>}
  }
  MaterialExpressionNamedRerouteDeclaration_33 {
   class=MaterialExpressionNamedRerouteDeclaration
   inputs [None {from=MaterialExpressionSubtract_10}]
   properties {name=DirectionBurnSide, nodeColor=<rgba 0.7344 1 0 1>}
  }
  MaterialExpressionNamedRerouteDeclaration_34 {
   class=MaterialExpressionNamedRerouteDeclaration
   inputs [None {from=MaterialExpressionTextureSampleParameter2D_10, output=G}]
   properties {name=Base_Texture_02_G, nodeColor=<rgba 1 0 0.9609 1>}
  }
  MaterialExpressionNamedRerouteDeclaration_35 {
   class=MaterialExpressionNamedRerouteDeclaration
   inputs [None {from=MaterialExpressionOneMinus_2}]
   properties {name=Emissive_Texture_02, nodeColor=<rgba 0.2188 0 1 1>}
  }
  MaterialExpressionNamedRerouteDeclaration_36 {
   class=MaterialExpressionNamedRerouteDeclaration
   inputs [None {from=MaterialExpressionSaturate_6}]
   properties {name=Dissolve_01, nodeColor=<rgba 0.0781 0 1 1>}
  }
  MaterialExpressionNamedRerouteDeclaration_38 {
   class=MaterialExpressionNamedRerouteDeclaration
   inputs [None {from=MaterialExpressionChannelMaskParameter_2, output=R}]
   properties {name=Emissive_Texture_01, nodeColor=<rgba 1 0.5156 0 1>}
  }
  MaterialExpressionNamedRerouteDeclaration_4 {
   class=MaterialExpressionNamedRerouteDeclaration
   inputs [None {from=MaterialExpressionMaterialFunctionCall_8, output=B}]
   properties {name=Metallic, nodeColor=<rgba 0.0122 0.0881 1 1>}
  }
  MaterialExpressionNamedRerouteDeclaration_5 {
   class=MaterialExpressionNamedRerouteDeclaration
   inputs [None {from=MaterialExpressionMaterialFunctionCall_1, output=Result}]
   properties {name=Dissolve, nodeColor=<rgba 0.1875 0.6953 1 1>}
  }
  MaterialExpressionNamedRerouteDeclaration_6 {
   class=MaterialExpressionNamedRerouteDeclaration
   inputs [None {from=MaterialExpressionAdd_10}]
   properties {name=Emissive, nodeColor=<rgba 1 0.5916 0 1>}
  }
  MaterialExpressionNamedRerouteDeclaration_7 {
   class=MaterialExpressionNamedRerouteDeclaration
   inputs [None {from=MaterialExpressionDynamicParameter_0, output=Emissive}]
   properties {name=DP Emissive, nodeColor=<rgba 0.7344 0 1 1>}
  }
  MaterialExpressionNamedRerouteDeclaration_8 {
   class=MaterialExpressionNamedRerouteDeclaration
   inputs [None {from=MaterialExpressionDynamicParameter_0, output=Metallic}]
   properties {name=DP Metallic, nodeColor=<rgba 0.7344 0 1 1>}
  }
  MaterialExpressionNamedRerouteDeclaration_9 {
   class=MaterialExpressionNamedRerouteDeclaration
   inputs [None {from=MaterialExpressionDynamicParameter_0, output=Dissolve}]
   properties {name=DP Dissolve, nodeColor=<rgba 0.7344 0 1 1>}
  }
  MaterialExpressionNamedRerouteUsage_0 {class=MaterialExpressionNamedRerouteUsage}
  MaterialExpressionNamedRerouteUsage_1 {class=MaterialExpressionNamedRerouteUsage}
  MaterialExpressionNamedRerouteUsage_10 {class=MaterialExpressionNamedRerouteUsage}
  MaterialExpressionNamedRerouteUsage_11 {class=MaterialExpressionNamedRerouteUsage}
  MaterialExpressionNamedRerouteUsage_12 {class=MaterialExpressionNamedRerouteUsage}
  MaterialExpressionNamedRerouteUsage_13 {class=MaterialExpressionNamedRerouteUsage}
  MaterialExpressionNamedRerouteUsage_14 {class=MaterialExpressionNamedRerouteUsage}
  MaterialExpressionNamedRerouteUsage_15 {class=MaterialExpressionNamedRerouteUsage}
  MaterialExpressionNamedRerouteUsage_16 {class=MaterialExpressionNamedRerouteUsage}
  MaterialExpressionNamedRerouteUsage_17 {class=MaterialExpressionNamedRerouteUsage}
  MaterialExpressionNamedRerouteUsage_18 {class=MaterialExpressionNamedRerouteUsage}
  MaterialExpressionNamedRerouteUsage_19 {class=MaterialExpressionNamedRerouteUsage}
  MaterialExpressionNamedRerouteUsage_2 {class=MaterialExpressionNamedRerouteUsage}
  MaterialExpressionNamedRerouteUsage_20 {class=MaterialExpressionNamedRerouteUsage}
  MaterialExpressionNamedRerouteUsage_21 {class=MaterialExpressionNamedRerouteUsage}
  MaterialExpressionNamedRerouteUsage_22 {class=MaterialExpressionNamedRerouteUsage}
  MaterialExpressionNamedRerouteUsage_23 {class=MaterialExpressionNamedRerouteUsage}
  MaterialExpressionNamedRerouteUsage_24 {class=MaterialExpressionNamedRerouteUsage}
  MaterialExpressionNamedRerouteUsage_25 {class=MaterialExpressionNamedRerouteUsage}
  MaterialExpressionNamedRerouteUsage_26 {class=MaterialExpressionNamedRerouteUsage}
  MaterialExpressionNamedRerouteUsage_27 {class=MaterialExpressionNamedRerouteUsage}
  MaterialExpressionNamedRerouteUsage_28 {class=MaterialExpressionNamedRerouteUsage}
  MaterialExpressionNamedRerouteUsage_29 {class=MaterialExpressionNamedRerouteUsage}
  MaterialExpressionNamedRerouteUsage_3 {class=MaterialExpressionNamedRerouteUsage}
  MaterialExpressionNamedRerouteUsage_30 {class=MaterialExpressionNamedRerouteUsage}
  MaterialExpressionNamedRerouteUsage_31 {class=MaterialExpressionNamedRerouteUsage}
  MaterialExpressionNamedRerouteUsage_32 {class=MaterialExpressionNamedRerouteUsage}
  MaterialExpressionNamedRerouteUsage_33 {class=MaterialExpressionNamedRerouteUsage}
  MaterialExpressionNamedRerouteUsage_34 {class=MaterialExpressionNamedRerouteUsage}
  MaterialExpressionNamedRerouteUsage_35 {class=MaterialExpressionNamedRerouteUsage}
  MaterialExpressionNamedRerouteUsage_36 {class=MaterialExpressionNamedRerouteUsage}
  MaterialExpressionNamedRerouteUsage_37 {class=MaterialExpressionNamedRerouteUsage}
  MaterialExpressionNamedRerouteUsage_38 {class=MaterialExpressionNamedRerouteUsage}
  MaterialExpressionNamedRerouteUsage_39 {class=MaterialExpressionNamedRerouteUsage}
  MaterialExpressionNamedRerouteUsage_4 {class=MaterialExpressionNamedRerouteUsage}
  MaterialExpressionNamedRerouteUsage_40 {class=MaterialExpressionNamedRerouteUsage}
  MaterialExpressionNamedRerouteUsage_41 {class=MaterialExpressionNamedRerouteUsage}
  MaterialExpressionNamedRerouteUsage_42 {class=MaterialExpressionNamedRerouteUsage}
  MaterialExpressionNamedRerouteUsage_43 {class=MaterialExpressionNamedRerouteUsage}
  MaterialExpressionNamedRerouteUsage_5 {class=MaterialExpressionNamedRerouteUsage}
  MaterialExpressionNamedRerouteUsage_6 {class=MaterialExpressionNamedRerouteUsage}
  MaterialExpressionNamedRerouteUsage_7 {class=MaterialExpressionNamedRerouteUsage}
  MaterialExpressionNamedRerouteUsage_8 {class=MaterialExpressionNamedRerouteUsage}
  MaterialExpressionNamedRerouteUsage_9 {class=MaterialExpressionNamedRerouteUsage}
  MaterialExpressionNormalize_0 {class=MaterialExpressionNormalize, inputs [VectorInput {from=MaterialExpressionTransformPosition_0, output=""}]}
  MaterialExpressionOneMinus_1 {class=MaterialExpressionOneMinus, inputs [None {from=MaterialExpressionTextureSampleParameter2D_7, output=R}]}
  MaterialExpressionOneMinus_2 {class=MaterialExpressionOneMinus, inputs [None {from=MaterialExpressionIf_13}]}
  MaterialExpressionOneMinus_4 {class=MaterialExpressionOneMinus, inputs [None {from=MaterialExpressionNamedRerouteUsage_23}]}
  MaterialExpressionParticleColor_0 {class=MaterialExpressionParticleColor, inputs [], properties {}}
  MaterialExpressionParticleColor_1 {class=MaterialExpressionParticleColor, inputs [], properties {}}
  MaterialExpressionParticleColor_2 {class=MaterialExpressionParticleColor, inputs [], properties {}}
  MaterialExpressionParticleColor_4 {class=MaterialExpressionParticleColor, inputs [], properties {}}
  MaterialExpressionParticleColor_5 {class=MaterialExpressionParticleColor, inputs [], properties {}}
  MaterialExpressionParticleSubUVProperties_0 {class=MaterialExpressionParticleSubUVProperties, inputs [], properties {}}
  MaterialExpressionPower_0 {
   class=MaterialExpressionPower
   inputs [
    Base {from=MaterialExpressionChannelMaskParameter_5, output=R}
    Exp {from=MaterialExpressionScalarParameter_1}
   ]
  }
  MaterialExpressionPower_1 {
   class=MaterialExpressionPower
   inputs [
    Base {from=MaterialExpressionAdd_5}
    Exp {from=MaterialExpressionScalarParameter_9}
   ]
  }
  MaterialExpressionPower_2 {
   class=MaterialExpressionPower
   inputs [
    Base {from=MaterialExpressionComponentMask_3}
    Exp {from=MaterialExpressionScalarParameter_25}
   ]
  }
  MaterialExpressionPower_3 {
   class=MaterialExpressionPower
   inputs [
    Base {from=MaterialExpressionTextureSampleParameter2D_5, output=RGB}
    Exp {from=MaterialExpressionScalarParameter_27}
   ]
  }
  MaterialExpressionPower_4 {
   class=MaterialExpressionPower
   inputs [
    Base {from=MaterialExpressionDotProduct_2}
    Exp {from=MaterialExpressionScalarParameter_36}
   ]
   properties {constExponent=5}
  }
  MaterialExpressionPower_5 {
   class=MaterialExpressionPower
   inputs [
    Base {from=MaterialExpressionIf_5}
    Exp {from=MaterialExpressionScalarParameter_58}
   ]
  }
  MaterialExpressionReroute_0 {class=MaterialExpressionReroute, inputs [None {from=MaterialExpressionScalarParameter_8}]}
  MaterialExpressionReroute_1 {class=MaterialExpressionReroute, inputs [None {from=MaterialExpressionStaticSwitchParameter_1}]}
  MaterialExpressionReroute_11 {class=MaterialExpressionReroute, inputs [None {from=""}]}
  MaterialExpressionReroute_2 {class=MaterialExpressionReroute, inputs [None {from=MaterialExpressionParticleColor_4, output=RGB}]}
  MaterialExpressionReroute_3 {class=MaterialExpressionReroute, inputs [None {from=MaterialExpressionNamedRerouteUsage_23}]}
  MaterialExpressionReroute_4 {class=MaterialExpressionReroute, inputs [None {from=MaterialExpressionLinearInterpolate_6}]}
  MaterialExpressionReroute_5 {class=MaterialExpressionReroute, inputs [None {from=MaterialExpressionScalarParameter_70}]}
  MaterialExpressionReroute_6 {class=MaterialExpressionReroute, inputs [None {from=MaterialExpressionParticleColor_5, output=RGB}]}
  MaterialExpressionReroute_7 {class=MaterialExpressionReroute, inputs [None {from=MaterialExpressionVectorParameter_15, output=RGB}]}
  MaterialExpressionReroute_8 {class=MaterialExpressionReroute, inputs [None {from=MaterialExpressionVectorParameter_14, output=RGB}]}
  MaterialExpressionRound_0 {class=MaterialExpressionRound, inputs [None {from=MaterialExpressionScalarParameter_0, output=""}], properties {}}
  MaterialExpressionRound_2 {class=MaterialExpressionRound, inputs [None {from=MaterialExpressionPower_4, output=""}], properties {}}
  MaterialExpressionSaturate_0 {class=MaterialExpressionSaturate, inputs [None {from=MaterialExpressionMultiply_17}]}
  MaterialExpressionSaturate_1 {class=MaterialExpressionSaturate, inputs [None {from=MaterialExpressionSubtract_1}]}
  MaterialExpressionSaturate_12 {class=MaterialExpressionSaturate, inputs [None {from=MaterialExpressionSmoothStep_5}]}
  MaterialExpressionSaturate_13 {class=MaterialExpressionSaturate, inputs [None {from=MaterialExpressionSmoothStep_6}]}
  MaterialExpressionSaturate_16 {class=MaterialExpressionSaturate, inputs [None {from=MaterialExpressionCeil_4}]}
  MaterialExpressionSaturate_17 {class=MaterialExpressionSaturate, inputs [None {from=MaterialExpressionCeil_7}]}
  MaterialExpressionSaturate_2 {class=MaterialExpressionSaturate, inputs [None {from=MaterialExpressionMultiply_34}]}
  MaterialExpressionSaturate_6 {class=MaterialExpressionSaturate, inputs [None {from=MaterialExpressionCeil_2}]}
  MaterialExpressionSaturate_7 {class=MaterialExpressionSaturate, inputs [None {from=MaterialExpressionCeil_3}]}
  MaterialExpressionScalarParameter_0 {
   class=MaterialExpressionScalarParameter
   properties {defaultValue=1, desc="0=False 1=True", group=03 Emissive, parameterName=AddBaseColorToEmissive?, sliderMax=1, sortPriority=13}
  }
  MaterialExpressionScalarParameter_1 {class=MaterialExpressionScalarParameter, properties {defaultValue=2, group=06 WPO, parameterName=WPO Spread}}
  MaterialExpressionScalarParameter_10 {class=MaterialExpressionScalarParameter, properties {defaultValue=2, group=03 Emissive, parameterName=Emissive - FakeShadow SmoothStep}}
  MaterialExpressionScalarParameter_11 {class=MaterialExpressionScalarParameter, properties {group=03 Emissive, parameterName=Emissive - HasShadows?, sliderMax=1}}
  MaterialExpressionScalarParameter_12 {class=MaterialExpressionScalarParameter, properties {group=03 Emissive, parameterName=Emissive - ShadowsHaveColor?, sliderMax=1}}
  MaterialExpressionScalarParameter_13 {class=MaterialExpressionScalarParameter, properties {defaultValue=1, group=03 Emissive, parameterName=Emissive - FakeShadow Multiply}}
  MaterialExpressionScalarParameter_14 {class=MaterialExpressionScalarParameter, properties {group=03 Emissive, parameterName=Emissive - HasCustomLightDirection?, sliderMax=1}}
  MaterialExpressionScalarParameter_15 {class=MaterialExpressionScalarParameter, properties {defaultValue=1, group=05 Normal, parameterName=Normal - Intensity}}
  MaterialExpressionScalarParameter_16 {class=MaterialExpressionScalarParameter, properties {group=04 Opacity, parameterName=Dissolve - UV Channel}}
  MaterialExpressionScalarParameter_17 {class=MaterialExpressionScalarParameter, properties {defaultValue=1024, group=04 Opacity, parameterName=Opacity - Dissolve Highlight Gain}}
  MaterialExpressionScalarParameter_18 {class=MaterialExpressionScalarParameter, properties {group=04 Opacity, parameterName=Opacity - Depth Fade}}
  MaterialExpressionScalarParameter_19 {class=MaterialExpressionScalarParameter, properties {group=04 Opacity, parameterName="Dissolve -  Edge Width"}}
  MaterialExpressionScalarParameter_2 {class=MaterialExpressionScalarParameter, properties {defaultValue=1, group=06 WPO, parameterName=WPO Strength}}
  MaterialExpressionScalarParameter_20 {class=MaterialExpressionScalarParameter, properties {defaultValue=1, parameterName=WPOFromNormalInt}}
  MaterialExpressionScalarParameter_21 {class=MaterialExpressionScalarParameter, properties {parameterName=WPOfromNormalOnly}}
  MaterialExpressionScalarParameter_22 {class=MaterialExpressionScalarParameter, properties {group=04 Opacity, parameterName=ManualDissolve}}
  MaterialExpressionScalarParameter_23 {class=MaterialExpressionScalarParameter, properties {defaultValue=1, group=04 Opacity, parameterName=Opacity - Use Dither, sliderMax=1}}
  MaterialExpressionScalarParameter_24 {class=MaterialExpressionScalarParameter, properties {parameterName=Base Color - UseSubUV}}
  MaterialExpressionScalarParameter_25 {class=MaterialExpressionScalarParameter, properties {group=Dissolve, parameterName=DissolveMaskPower}}
  MaterialExpressionScalarParameter_26 {class=MaterialExpressionScalarParameter, properties {defaultValue=1, group=06 WPO, parameterName=WPO ON/OFF, sortPriority=1}}
  MaterialExpressionScalarParameter_27 {class=MaterialExpressionScalarParameter, properties {defaultValue=1, group=07 Matcap, parameterName=Matcap - Power, sortPriority=33}}
  MaterialExpressionScalarParameter_28 {class=MaterialExpressionScalarParameter, properties {group=07 Matcap, parameterName=Matcap - Intensity, sortPriority=31}}
  MaterialExpressionScalarParameter_29 {class=MaterialExpressionScalarParameter, properties {group=07 Matcap, parameterName=Matcap - Intensity, sortPriority=31}}
  MaterialExpressionScalarParameter_3 {class=MaterialExpressionScalarParameter, properties {defaultValue=0.3, group=03 Emissive, parameterName=Emissive - FakeShadow MoveGradient}}
  MaterialExpressionScalarParameter_30 {class=MaterialExpressionScalarParameter, properties {parameterName="Color SMR (1 Yes 0 No)"}}
  MaterialExpressionScalarParameter_31 {class=MaterialExpressionScalarParameter, properties {defaultValue=1, group=03 Emissive, parameterName=Texture Intensity, sortPriority=1}}
  MaterialExpressionScalarParameter_32 {class=MaterialExpressionScalarParameter, properties {group=03 Emissive, parameterName=Emissive - Fresnel Min, sortPriority=11}}
  MaterialExpressionScalarParameter_33 {class=MaterialExpressionScalarParameter, properties {group=03 Emissive, parameterName=Emissive - Fresnel Max, sortPriority=12}}
  MaterialExpressionScalarParameter_34 {class=MaterialExpressionScalarParameter, properties {group=01 Base Color, parameterName=Base_PaintVarient, sortPriority=2}}
  MaterialExpressionScalarParameter_35 {class=MaterialExpressionScalarParameter, properties {group=01 Base Color, parameterName=BaseColorSwitch, sortPriority=3}}
  MaterialExpressionScalarParameter_36 {class=MaterialExpressionScalarParameter, properties {defaultValue=2, group=08 Add Highlight, parameterName=Power, sortPriority=1}}
  MaterialExpressionScalarParameter_37 {class=MaterialExpressionScalarParameter, properties {parameterName=AnimateDissolveMask}}
  MaterialExpressionScalarParameter_38 {class=MaterialExpressionScalarParameter, properties {group=08 Add Highlight, parameterName=Emissive_PaintVarient, sortPriority=2}}
  MaterialExpressionScalarParameter_39 {class=MaterialExpressionScalarParameter, properties {group=08 Add Highlight, parameterName=Add Highlight, sortPriority=2}}
  MaterialExpressionScalarParameter_4 {
   class=MaterialExpressionScalarParameter
   properties {defaultValue=1, group=03 Emissive, parameterName=Emissive - BaseColor Multiplier, sortPriority=15}
  }
  MaterialExpressionScalarParameter_40 {class=MaterialExpressionScalarParameter, properties {parameterName=UseBaseTexture_02}}
  MaterialExpressionScalarParameter_41 {class=MaterialExpressionScalarParameter, properties {parameterName=BaseColor_UesDissolve}}
  MaterialExpressionScalarParameter_42 {
   class=MaterialExpressionScalarParameter
   properties {defaultValue=-0.5, group=09 Directional Dissolve, parameterName=TransitionEnd, sortPriority=44}
  }
  MaterialExpressionScalarParameter_43 {class=MaterialExpressionScalarParameter, properties {group=03 Emissive, parameterName=Texture 2 Pan U, sortPriority=36}}
  MaterialExpressionScalarParameter_44 {class=MaterialExpressionScalarParameter, properties {group=03 Emissive, parameterName=Texture 2 Pan V, sortPriority=37}}
  MaterialExpressionScalarParameter_45 {
   class=MaterialExpressionScalarParameter
   properties {defaultValue=512, group=09 Directional Dissolve, parameterName=Burn Brightness, sortPriority=39}
  }
  MaterialExpressionScalarParameter_46 {class=MaterialExpressionScalarParameter, properties {group=03 Emissive, parameterName=Emissive02 - Use Special Texture?}}
  MaterialExpressionScalarParameter_47 {class=MaterialExpressionScalarParameter, properties {group=03 Emissive, parameterName=Emissive Tex Offset U, sortPriority=3}}
  MaterialExpressionScalarParameter_48 {class=MaterialExpressionScalarParameter, properties {group=03 Emissive, parameterName=Emissive Tex Offset V, sortPriority=4}}
  MaterialExpressionScalarParameter_49 {class=MaterialExpressionScalarParameter, properties {defaultValue=1, parameterName=UseFresnelMask}}
  MaterialExpressionScalarParameter_5 {
   class=MaterialExpressionScalarParameter
   properties {defaultValue=1, group=09 Directional Dissolve, parameterName=DirectionMaskOneMinus, sortPriority=8}
  }
  MaterialExpressionScalarParameter_50 {
   class=MaterialExpressionScalarParameter
   properties {defaultValue=5, group=03 Emissive, parameterName=Emissive - Fresnel Exp, sortPriority=10}
  }
  MaterialExpressionScalarParameter_51 {class=MaterialExpressionScalarParameter, properties {parameterName="Use_UV[1]"}}
  MaterialExpressionScalarParameter_52 {class=MaterialExpressionScalarParameter, properties {group=03 Emissive, parameterName=UseEmissiveTexture_02, sortPriority=0}}
  MaterialExpressionScalarParameter_53 {class=MaterialExpressionScalarParameter, properties {defaultValue=-0.1, group=09 Directional Dissolve, parameterName=Burn, sortPriority=37}}
  MaterialExpressionScalarParameter_54 {
   class=MaterialExpressionScalarParameter
   properties {defaultValue=0.05, group=09 Directional Dissolve, parameterName=BurnSideWith, sortPriority=38}
  }
  MaterialExpressionScalarParameter_55 {class=MaterialExpressionScalarParameter, properties {group=03 Emissive, parameterName=Switch_Emissive, sortPriority=34}}
  MaterialExpressionScalarParameter_56 {class=MaterialExpressionScalarParameter, properties {group=09 Directional Dissolve, parameterName=Switch_Mask, sortPriority=0}}
  MaterialExpressionScalarParameter_57 {class=MaterialExpressionScalarParameter, properties {group=03 Emissive, parameterName=Add Emissive Texture End, sortPriority=5}}
  MaterialExpressionScalarParameter_58 {class=MaterialExpressionScalarParameter, properties {defaultValue=1, parameterName=Base Color Power}}
  MaterialExpressionScalarParameter_6 {class=MaterialExpressionScalarParameter, properties {defaultValue=20, group=09 Directional Dissolve, parameterName=TexCoordDirectionMask}}
  MaterialExpressionScalarParameter_64 {
   class=MaterialExpressionScalarParameter
   properties {defaultValue=0.5, group=09 Directional Dissolve, parameterName=FresnelDissolve, sortPriority=36}
  }
  MaterialExpressionScalarParameter_65 {class=MaterialExpressionScalarParameter, properties {defaultValue=0.5, group=09 Directional Dissolve, parameterName=FresnelFeather}}
  MaterialExpressionScalarParameter_66 {
   class=MaterialExpressionScalarParameter
   properties {defaultValue=0.6, group=09 Directional Dissolve, parameterName=FireDissolve, sortPriority=28}
  }
  MaterialExpressionScalarParameter_67 {
   class=MaterialExpressionScalarParameter
   properties {defaultValue=0.1, group=09 Directional Dissolve, parameterName=FireFeather, sortPriority=24}
  }
  MaterialExpressionScalarParameter_68 {class=MaterialExpressionScalarParameter, properties {group=09 Directional Dissolve, parameterName=DirectionDissolve, sortPriority=4}}
  MaterialExpressionScalarParameter_69 {
   class=MaterialExpressionScalarParameter
   properties {defaultValue=0.01, group=09 Directional Dissolve, parameterName=DirectionBurnSideWith, sortPriority=12}
  }
  MaterialExpressionScalarParameter_7 {
   class=MaterialExpressionScalarParameter
   properties {defaultValue=1, group=09 Directional Dissolve, parameterName=TexCoordIndex, sortPriority=22}
  }
  MaterialExpressionScalarParameter_70 {
   class=MaterialExpressionScalarParameter
   properties {defaultValue=8, group=09 Directional Dissolve, parameterName=DirectionDissolveSmooth, sortPriority=16}
  }
  MaterialExpressionScalarParameter_71 {class=MaterialExpressionScalarParameter, properties {group=03 Emissive, parameterName=EmissiveBaseTextureIntensity}}
  MaterialExpressionScalarParameter_8 {class=MaterialExpressionScalarParameter, properties {defaultValue=1, parameterName=Emissive Color Fresnel, sliderMax=1}}
  MaterialExpressionScalarParameter_9 {class=MaterialExpressionScalarParameter, properties {defaultValue=1, group=03 Emissive, parameterName=Emissive - FakeShadow Power}}
  MaterialExpressionSmoothStep_0 {
   class=MaterialExpressionSmoothStep
   inputs [
    Min {from="", output=""}
    Max {from=MaterialExpressionScalarParameter_10, output=""}
    Value {from=MaterialExpressionPower_1, output=""}
   ]
   properties {}
  }
  MaterialExpressionSmoothStep_5 {
   class=MaterialExpressionSmoothStep
   inputs [
    Min {from=MaterialExpressionScalarParameter_64, output=""}
    Max {from=MaterialExpressionAdd_22, output=""}
    Value {from=MaterialExpressionNamedRerouteUsage_29, output=""}
   ]
   properties {}
  }
  MaterialExpressionSmoothStep_6 {
   class=MaterialExpressionSmoothStep
   inputs [
    Min {from=MaterialExpressionScalarParameter_66, output=""}
    Max {from=MaterialExpressionAdd_23, output=""}
    Value {from=MaterialExpressionNamedRerouteUsage_20, output=""}
   ]
   properties {}
  }
  MaterialExpressionStaticSwitchParameter_0 {
   class=MaterialExpressionStaticSwitchParameter
   inputs [
    True {from=MaterialExpressionMultiply_27, output=""}
    False {from=MaterialExpressionConstant_1, output=""}
   ]
   properties {group=03 Emissive, parameterName=UseEmissiveTexture, sortPriority=0}
  }
  MaterialExpressionStaticSwitchParameter_1 {
   class=MaterialExpressionStaticSwitchParameter
   inputs [
    True {from=MaterialExpressionMultiply_6, output=""}
    False {from=MaterialExpressionNamedRerouteUsage_5, output=""}
   ]
   properties {
    defaultValue=true
    desc="Uses Base Texture's A channel for opacity mask by default (true)"
    group=04 Opacity
    parameterName=Opacity Mask-UseBaseTextureAlpha
   }
  }
  MaterialExpressionStaticSwitchParameter_2 {
   class=MaterialExpressionStaticSwitchParameter
   inputs [
    True {from=MaterialExpressionMultiply_20, output=""}
    False {from=MaterialExpressionReroute_1, output=""}
   ]
   properties {
    desc="Uses Base Texture's A channel for opacity mask by default (true)"
    group=04 Opacity
    parameterName=Opacity Mask-UseFullScreenTransition
   }
  }
  MaterialExpressionStaticSwitchParameter_3 {
   class=MaterialExpressionStaticSwitchParameter
   inputs [
    True {from=MaterialExpressionMultiply_22, output=""}
    False {from=MaterialExpressionStaticSwitchParameter_2, output=""}
   ]
   properties {
    desc="Uses Base Texture's A channel for opacity mask by default (true)"
    group=04 Opacity
    parameterName=Opacity Mask- Multiply Particle Alpha with Opacity
   }
  }
  MaterialExpressionSubtract_0 {
   class=MaterialExpressionSubtract
   inputs [
    A {from=MaterialExpressionSaturate_2}
    B {from=MaterialExpressionLinearInterpolate_23}
   ]
  }
  MaterialExpressionSubtract_1 {
   class=MaterialExpressionSubtract
   inputs [
    A {from=MaterialExpressionOneMinus_1}
    B {from=MaterialExpressionScalarParameter_42}
   ]
  }
  MaterialExpressionSubtract_10 {
   class=MaterialExpressionSubtract
   inputs [
    A {from=MaterialExpressionSaturate_16}
    B {from=MaterialExpressionSaturate_17}
   ]
  }
  MaterialExpressionSubtract_3 {
   class=MaterialExpressionSubtract
   inputs [
    A {from=MaterialExpressionSaturate_2}
    B {from=MaterialExpressionLinearInterpolate_24}
   ]
  }
  MaterialExpressionSubtract_4 {
   class=MaterialExpressionSubtract
   inputs [
    A {from=MaterialExpressionNamedRerouteDeclaration_36}
    B {from=MaterialExpressionSaturate_7}
   ]
  }
  MaterialExpressionSubtract_8 {
   class=MaterialExpressionSubtract
   inputs [
    A {from=MaterialExpressionLinearInterpolate_4}
    B {from=MaterialExpressionScalarParameter_68}
   ]
  }
  MaterialExpressionSubtract_9 {
   class=MaterialExpressionSubtract
   inputs [
    A {from=MaterialExpressionLinearInterpolate_4}
    B {from=MaterialExpressionAdd_24}
   ]
  }
  MaterialExpressionSwitch_0 {
   class=MaterialExpressionSwitch
   inputs [
    SwitchValue {from=MaterialExpressionScalarParameter_7, output=""}
    Default {from="", output=""}
    None {from=MaterialExpressionTextureCoordinate_11, output=""}
    None {from=MaterialExpressionTextureCoordinate_12, output=""}
   ]
   properties {inputs [
     {
      inputName=None
      input {
       expression=&self:MaterialExpressionTextureCoordinate_11
       inputName=None
       mask=0
       maskA=0
       maskB=0
       maskG=0
       maskR=0
       outputIndex=0
      }
     }
     {
      inputName=None
      input {
       expression=&self:MaterialExpressionTextureCoordinate_12
       inputName=None
       mask=0
       maskA=0
       maskB=0
       maskG=0
       maskR=0
       outputIndex=0
      }
     }
    ]}
  }
  MaterialExpressionTextureCoordinate_0 {class=MaterialExpressionTextureCoordinate}
  MaterialExpressionTextureCoordinate_1 {class=MaterialExpressionTextureCoordinate}
  MaterialExpressionTextureCoordinate_11 {class=MaterialExpressionTextureCoordinate}
  MaterialExpressionTextureCoordinate_12 {class=MaterialExpressionTextureCoordinate, properties {coordinateIndex=1}}
  MaterialExpressionTextureCoordinate_2 {class=MaterialExpressionTextureCoordinate}
  MaterialExpressionTextureCoordinate_3 {class=MaterialExpressionTextureCoordinate}
  MaterialExpressionTextureCoordinate_4 {class=MaterialExpressionTextureCoordinate, properties {coordinateIndex=1}}
  MaterialExpressionTextureCoordinate_5 {class=MaterialExpressionTextureCoordinate}
  MaterialExpressionTextureCoordinate_6 {class=MaterialExpressionTextureCoordinate}
  MaterialExpressionTextureCoordinate_7 {class=MaterialExpressionTextureCoordinate, properties {coordinateIndex=1, uTiling=16, vTiling=16}}
  MaterialExpressionTextureCoordinate_9 {class=MaterialExpressionTextureCoordinate, properties {coordinateIndex=1}}
  MaterialExpressionTextureSampleParameter2D_0 {
   class=MaterialExpressionTextureSampleParameter2D
   inputs [
    UVs {from=MaterialExpressionIf_2, output=""}
    Apply View MipBias {from="", output=""}
   ]
   properties {group=01 Base Color, parameterName=Base Color Texture, sortPriority=1, texture=/Game/Global/Textures/Default/Blanks/T_MidgreysRGB}
  }
  MaterialExpressionTextureSampleParameter2D_1 {
   class=MaterialExpressionTextureSampleParameter2D
   inputs [
    UVs {from=MaterialExpressionNamedRerouteUsage_19, output=""}
    Apply View MipBias {from="", output=""}
   ]
   properties {
    group=05 Normal
    parameterName=Normal - Texture
    samplerType=SAMPLERTYPE_Normal
    sortPriority=0
    texture=&FlatNormalBlue
   }
  }
  MaterialExpressionTextureSampleParameter2D_10 {
   class=MaterialExpressionTextureSampleParameter2D
   inputs [
    UVs {from="", output=""}
    Apply View MipBias {from="", output=""}
   ]
   properties {group=01 Base Color, parameterName=Base Color Texture_02, sortPriority=1, texture=&T_MidgreysRGB}
  }
  MaterialExpressionTextureSampleParameter2D_11 {
   class=MaterialExpressionTextureSampleParameter2D
   inputs [
    UVs {from=MaterialExpressionAdd_11, output=""}
    Apply View MipBias {from="", output=""}
   ]
   properties {group=03 Emissive, parameterName=Emissive - Texture_02, sortPriority=1, texture=&T_BlacksRGB}
  }
  MaterialExpressionTextureSampleParameter2D_2 {
   class=MaterialExpressionTextureSampleParameter2D
   inputs [
    UVs {from=MaterialExpressionNamedRerouteUsage_18, output=""}
    Apply View MipBias {from="", output=""}
   ]
   properties {
    group=02 SRM
    parameterName=SRM - Texture
    samplerType=SAMPLERTYPE_Masks
    sortPriority=0
    texture=&T_BlackMask
   }
  }
  MaterialExpressionTextureSampleParameter2D_3 {
   class=MaterialExpressionTextureSampleParameter2D
   inputs [
    UVs {from=MaterialExpressionIf_12, output=""}
    Apply View MipBias {from="", output=""}
   ]
   properties {group=03 Emissive, parameterName=Emissive - Texture, sortPriority=1, texture=&T_BlacksRGB}
  }
  MaterialExpressionTextureSampleParameter2D_4 {
   class=MaterialExpressionTextureSampleParameter2D
   inputs [
    UVs {from=MaterialExpressionAdd_0, output=""}
    Apply View MipBias {from="", output=""}
   ]
   properties {
    group=04 Opacity
    parameterName=Dissolve - Texture
    samplerType=SAMPLERTYPE_LinearColor
    sortPriority=30
    texture=/Game/Global/Textures/Default/Noise/T_CloudNoise_01
   }
  }
  MaterialExpressionTextureSampleParameter2D_5 {
   class=MaterialExpressionTextureSampleParameter2D
   inputs [
    UVs {from=MaterialExpressionAdd_8, output=""}
    Apply View MipBias {from="", output=""}
   ]
   properties {
    group=07 Matcap
    parameterName=Matcap Texture
    texture=/Game/Global/Materials/Character_Parent_2023/Anime_Parent_2023/Textures/T_MatCap_Highlight
   }
  }
  MaterialExpressionTextureSampleParameter2D_6 {
   class=MaterialExpressionTextureSampleParameter2D
   inputs [
    UVs {from=MaterialExpressionAdd_9, output=""}
    Apply View MipBias {from="", output=""}
   ]
   properties {group=03 Emissive, parameterName="Special  Texture", sortPriority=1, texture=&T_MidgreysRGB}
  }
  MaterialExpressionTextureSampleParameter2D_7 {
   class=MaterialExpressionTextureSampleParameter2D
   inputs [
    UVs {from=MaterialExpressionTextureCoordinate_7, output=""}
    Apply View MipBias {from="", output=""}
   ]
   properties {group=09 Directional Dissolve, parameterName=CellNoise, sortPriority=40, texture=/Game/Global/Textures/Default/Noise/T_CellNoise_01}
  }
  MaterialExpressionTime_0 {class=MaterialExpressionTime}
  MaterialExpressionTime_1 {class=MaterialExpressionTime}
  MaterialExpressionTime_2 {class=MaterialExpressionTime}
  MaterialExpressionTime_3 {class=MaterialExpressionTime}
  MaterialExpressionTransformPosition_0 {
   class=MaterialExpressionTransformPosition
   inputs [None {from=MaterialExpressionWorldPosition_0, output=XYZ}]
   properties {transformSourceType=TRANSFORMPOSSOURCE_World, transformType=TRANSFORMPOSSOURCE_View}
  }
  MaterialExpressionTransform_0 {
   class=MaterialExpressionTransform
   inputs [None {from=MaterialExpressionVertexNormalWS_2}]
   properties {transformSourceType=TRANSFORMSOURCE_World, transformType=TRANSFORM_View}
  }
  MaterialExpressionVectorParameter_0 {
   class=MaterialExpressionVectorParameter
   properties {group=04 Opacity, parameterName=Dissolve - UVs, channelNames {a=Panner Y, b=Panner X, g=Scale Y, r=Scale X}, defaultValue=<rgba 1 1 0 0>}
  }
  MaterialExpressionVectorParameter_1 {class=MaterialExpressionVectorParameter, properties {group=01 Base Color, parameterName=Base Color - Tint, defaultValue=<rgba 1 1 1 1>}}
  MaterialExpressionVectorParameter_10 {class=MaterialExpressionVectorParameter, properties {parameterName=SMR}}
  MaterialExpressionVectorParameter_11 {class=MaterialExpressionVectorParameter, properties {group=03 Emissive, parameterName=Special Texture - Tint}}
  MaterialExpressionVectorParameter_12 {
   class=MaterialExpressionVectorParameter
   properties {
    group=03 Emissive
    parameterName=Special Texture - UVs
    channelNames {a=Panner Y, b=Panner X, g=Scale Y, r=Scale X}
    defaultValue=<rgba 1 1 0 0>
   }
  }
  MaterialExpressionVectorParameter_14 {
   class=MaterialExpressionVectorParameter
   properties {group=08 Add Highlight, parameterName=Defauit_1, sortPriority=0, defaultValue=<rgba 1 1 1 1>}
  }
  MaterialExpressionVectorParameter_15 {
   class=MaterialExpressionVectorParameter
   properties {group=08 Add Highlight, parameterName=Defauit_2, sortPriority=0, defaultValue=<rgba 1 1 1 1>}
  }
  MaterialExpressionVectorParameter_16 {
   class=MaterialExpressionVectorParameter
   properties {
    group=03 Emissive
    parameterName="Emissive - UV[1]"
    sortPriority=35
    channelNames {a=Texture_2_V, b=Texture_2_U, g=Texture_1_V, r=Texture_1_U}
    defaultValue=<rgba 1 1 1 1>
   }
  }
  MaterialExpressionVectorParameter_17 {
   class=MaterialExpressionVectorParameter
   properties {group=03 Emissive, parameterName=FresnelColor, sortPriority=35, defaultValue=<rgba 1 1 1 1>}
  }
  MaterialExpressionVectorParameter_2 {
   class=MaterialExpressionVectorParameter
   properties {
    group=03 Emissive
    parameterName=Emissive - UVs
    sortPriority=2
    channelNames {a=Panner Y, b=Panner X, g=Scale Y, r=Scale X}
    defaultValue=<rgba 1 1 0 0>
   }
  }
  MaterialExpressionVectorParameter_3 {
   class=MaterialExpressionVectorParameter
   properties {
    group=01 Base Color
    parameterName=Base Color - UVs
    channelNames {a=Panner Y, b=Panner X, g=Scale Y, r=Scale X}
    defaultValue=<rgba 1 1 0 0>
   }
  }
  MaterialExpressionVectorParameter_4 {
   class=MaterialExpressionVectorParameter
   properties {group=03 Emissive, parameterName=FresnelColor, sortPriority=35, defaultValue=<rgba 1 1 1 1>}
  }
  MaterialExpressionVectorParameter_5 {
   class=MaterialExpressionVectorParameter
   properties {group=03 Emissive, parameterName=Emissive - FakeShadow DarkColor, defaultValue=<rgba 0.0095 0.0115 0.0521 1>}
  }
  MaterialExpressionVectorParameter_6 {
   class=MaterialExpressionVectorParameter
   properties {group=03 Emissive, parameterName=Emissive - FakeShadow LightColor, defaultValue=<rgba 1 1 1 0>}
  }
  MaterialExpressionVectorParameter_7 {
   class=MaterialExpressionVectorParameter
   properties {parameterName=Emissive - FakeShadow CustomVectorLightDirection, defaultValue=<rgba 0 0 0 1>}
  }
  MaterialExpressionVectorParameter_8 {class=MaterialExpressionVectorParameter, properties {group=03 Emissive, parameterName=Emissive - Color, defaultValue=<rgba 1 1 1 1>}}
  MaterialExpressionVectorParameter_9 {
   class=MaterialExpressionVectorParameter
   properties {group=04 Opacity, parameterName=Opacity - Dissolve Highlight Color, defaultValue=<rgba 1 0.5 0.1 1>}
  }
  MaterialExpressionVertexNormalWS_0 {class=MaterialExpressionVertexNormalWS, inputs [], properties {}}
  MaterialExpressionVertexNormalWS_1 {class=MaterialExpressionVertexNormalWS, inputs [], properties {}}
  MaterialExpressionVertexNormalWS_2 {class=MaterialExpressionVertexNormalWS, inputs [], properties {}}
  MaterialExpressionVertexNormalWS_3 {class=MaterialExpressionVertexNormalWS, inputs [], properties {}}
  MaterialExpressionWorldPosition_0 {class=MaterialExpressionWorldPosition, inputs [], properties {}}
 }
 properties {
  bAutomaticallySetUsageInEditor=false
  bUsedWithMeshDeformer=true
  bUsedWithMeshParticles=true
  bUsedWithNiagaraMeshParticles=true
  bUsedWithSkeletalMesh=true
  blendMode=BLEND_Masked
  usedShadingModels=Default Lit
 }
}
